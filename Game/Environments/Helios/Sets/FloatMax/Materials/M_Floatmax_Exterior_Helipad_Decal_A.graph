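# Deferred decal material for the exterior helipad markings of the FloatMax set in the Helios environment.
# It uses Modulate blending, which multiplies the underlying colour rather than covering it.
# The graph holds a single texture sample of the helipad decal's _D texture, with no parameters or math nodes.
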Material /Game/Environments/Helios/Sets/FloatMax/Materials/M_Floatmax_Exterior_Helipad_Decal_A
graph {
 properties {blendMode=BLEND_Modulate, materialDomain=MD_DeferredDecal, pixelDepthOffsetMode=PDOM_Legacy, usedShadingModels=Default Lit}
 nodes [TextureSample() {texture=/Game/Environments/Helios/Sets/FloatMax/Textures/Decals/T_Floatmax_Exterior_Helipad_Decal_A_D}]
}
